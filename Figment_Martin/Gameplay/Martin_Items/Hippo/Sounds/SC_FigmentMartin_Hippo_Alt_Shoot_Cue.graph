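# Sound cue for the alternate-fire shot of the Hippo item in Figment Martin, using Ashton Hippo grapple shoot recordings.
# A Random node picks one of four equally weighted wave variations and feeds an Attenuation node using a close-range weapon projectile preset, under a SoundClass root.
# The cue slightly lowers volume and overrides concurrency so new instances are blocked when the limit is reached.
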SoundCue /Figment_Martin/Gameplay/Martin_Items/Hippo/Sounds/SC_FigmentMartin_Hippo_Alt_Shoot_Cue
root=SoundClass_3
nodes [
 {id=Attenuation_3, children [Random_1], properties {attenuationSettings=&Fort_Weapon_Projectile_Close_Athena_Att}}
 {id=Random_1, children [WavePlayer_0,WavePlayer_1,WavePlayer_2,WavePlayer_3], properties {weights [1,1,1,1]}}
 {id=SoundClass_3, children [Attenuation_3]}
 {id=WavePlayer_0, properties {soundWaveAssetPtr=&/Ashton_Hippo_Grapple_Shoot_01}}
 {id=WavePlayer_1, properties {soundWaveAssetPtr=&/Ashton_Hippo_Grapple_Shoot_02}}
 {id=WavePlayer_2, properties {soundWaveAssetPtr=&/Ashton_Hippo_Grapple_Shoot_03}}
 {id=WavePlayer_3, properties {soundWaveAssetPtr=&/Ashton_Hippo_Grapple_Shoot_04}}
]
properties {
 bOverrideConcurrency=true
 maxDistance=4950
 volumeMultiplier=0.8
 attenuationOverrides {falloffDistance=5000, attenuationShapeExtents {x=1000}}
 concurrencyOverrides {resolutionRule=PreventNew}
}
